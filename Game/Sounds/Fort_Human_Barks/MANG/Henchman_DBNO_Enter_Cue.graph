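# Sound cue for a MANG henchman voice bark, played on entering the down-but-not-out (DBNO) state.
# A Random node picks among six equally weighted wave variations and is excluded from branch culling.
# The cue uses close-range henchman attenuation, the non-critical enemy VO sound class, a slight volume boost and the shared MANG chatter concurrency.
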SoundCue /Game/Sounds/Fort_Human_Barks/MANG/Henchman_DBNO_Enter_Cue
root=Random_1
nodes [
 {
  id=Random_1
  children [WavePlayer_0, WavePlayer_7, WavePlayer_8, WavePlayer_9, WavePlayer_10, WavePlayer_11]
  properties {bShouldExcludeFromBranchCulling=true, weights [1,1,1,1,1,1]}
 }
 {id=WavePlayer_0, properties {soundWaveAssetPtr=&/Henchman_DBNO_Enter_010}}
 {id=WavePlayer_10, properties {soundWaveAssetPtr=&/Henchman_DBNO_Enter_050}}
 {id=WavePlayer_11, properties {soundWaveAssetPtr=&/Henchman_DBNO_Enter_060}}
 {id=WavePlayer_7, properties {soundWaveAssetPtr=&/Henchman_DBNO_Enter_020}}
 {id=WavePlayer_8, properties {soundWaveAssetPtr=&/Henchman_DBNO_Enter_030}}
 {id=WavePlayer_9, properties {soundWaveAssetPtr=&/Henchman_DBNO_Enter_040}}
]
properties {
 attenuationSettings=&/Henchman_MANG_Close_Att
 maxDistance=4000
 soundClassObject=/Game/Sounds/FortSoundClassesAndModes/Fort_VO_Enemy_NonCritical
 volumeMultiplier=1.2
 concurrencySet [&/MANG_Chatter_Conc]
}
