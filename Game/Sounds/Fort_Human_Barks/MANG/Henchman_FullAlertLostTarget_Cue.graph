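# Sound cue for henchman voice barks in the MANG set, named for losing a target while at full alert.
# A random node picks one of six equally weighted wave variations and is excluded from branch culling.
# Playback uses a close-range henchman attenuation, the non-critical enemy VO sound class and a shared chatter concurrency limit, with a slight volume boost.
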SoundCue /Game/Sounds/Fort_Human_Barks/MANG/Henchman_FullAlertLostTarget_Cue
root=Random_0
nodes [
 {
  id=Random_0
  children [WavePlayer_1,WavePlayer_2,WavePlayer_3,WavePlayer_4,WavePlayer_5,WavePlayer_6]
  properties {bShouldExcludeFromBranchCulling=true, weights [1,1,1,1,1,1]}
 }
 {id=WavePlayer_1, properties {soundWaveAssetPtr=&/Henchman_FullAlertLostTarget_010}}
 {id=WavePlayer_2, properties {soundWaveAssetPtr=&/Henchman_FullAlertLostTarget_020}}
 {id=WavePlayer_3, properties {soundWaveAssetPtr=&/Henchman_FullAlertLostTarget_030}}
 {id=WavePlayer_4, properties {soundWaveAssetPtr=&/Henchman_FullAlertLostTarget_040}}
 {id=WavePlayer_5, properties {soundWaveAssetPtr=&/Henchman_FullAlertLostTarget_050}}
 {id=WavePlayer_6, properties {soundWaveAssetPtr=&/Henchman_FullAlertLostTarget_060}}
]
properties {
 attenuationSettings=&/Henchman_MANG_Close_Att
 maxDistance=4000
 soundClassObject=/Game/Sounds/FortSoundClassesAndModes/Fort_VO_Enemy_NonCritical
 volumeMultiplier=1.2
 concurrencySet [&/MANG_Chatter_Conc]
}
