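# Sound cue for Henchman idle voice barks, classed as non-critical enemy VO.
# A Random root picks one of nine equally weighted Henchman idle wave variations and is excluded from branch culling.
# The cue gets a slight volume boost, uses the MANG close-range attenuation and shares a MANG chatter concurrency group.
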SoundCue /Game/Sounds/Fort_Human_Barks/MANG/Henchman_Idle_Cue
root=Random_0
nodes [
 {
  id=Random_0
  children [WavePlayer_2, WavePlayer_3, WavePlayer_4, WavePlayer_5, WavePlayer_6, WavePlayer_7, WavePlayer_8, WavePlayer_9, WavePlayer_10]
  properties {bShouldExcludeFromBranchCulling=true, weights [1,1,1,1,1,1,1,1,1]}
 }
 {id=WavePlayer_10, properties {soundWaveAssetPtr=&/Henchman_Idle_090}}
 {id=WavePlayer_2, properties {soundWaveAssetPtr=&/Henchman_Idle_010}}
 {id=WavePlayer_3, properties {soundWaveAssetPtr=&/Henchman_Idle_020}}
 {id=WavePlayer_4, properties {soundWaveAssetPtr=&/Henchman_Idle_030}}
 {id=WavePlayer_5, properties {soundWaveAssetPtr=&/Henchman_Idle_040}}
 {id=WavePlayer_6, properties {soundWaveAssetPtr=&/Henchman_Idle_050}}
 {id=WavePlayer_7, properties {soundWaveAssetPtr=&/Henchman_Idle_060}}
 {id=WavePlayer_8, properties {soundWaveAssetPtr=&/Henchman_Idle_070}}
 {id=WavePlayer_9, properties {soundWaveAssetPtr=&/Henchman_Idle_080}}
]
properties {
 attenuationSettings=&/Henchman_MANG_Close_Att
 maxDistance=4000
 soundClassObject=/Game/Sounds/FortSoundClassesAndModes/Fort_VO_Enemy_NonCritical
 volumeMultiplier=1.2
 concurrencySet [&/MANG_Chatter_Conc]
}
