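# Sound cue for henchman interrogation voice barks in the MANG human bark set.
# A random node picks one of nine interrogation line recordings with equal weight and is excluded from branch culling.
# Playback uses a close-range henchman attenuation, the non-critical enemy VO sound class and a shared MANG chatter concurrency group, with a slight volume boost.
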SoundCue /Game/Sounds/Fort_Human_Barks/MANG/Henchman_Interrogation_Cue
root=Random_1
nodes [
 {
  id=Random_1
  children [WavePlayer_0, WavePlayer_1, WavePlayer_11, WavePlayer_12, WavePlayer_13, WavePlayer_14, WavePlayer_15, WavePlayer_16, WavePlayer_17]
  properties {bShouldExcludeFromBranchCulling=true, weights [1,1,1,1,1,1,1,1,1]}
 }
 {id=WavePlayer_0, properties {soundWaveAssetPtr=&/Henchman_Interrogation_010}}
 {id=WavePlayer_1, properties {soundWaveAssetPtr=&/Henchman_Interrogation_020}}
 {id=WavePlayer_11, properties {soundWaveAssetPtr=&/Henchman_Interrogation_030}}
 {id=WavePlayer_12, properties {soundWaveAssetPtr=&/Henchman_Interrogation_040}}
 {id=WavePlayer_13, properties {soundWaveAssetPtr=&/Henchman_Interrogation_050}}
 {id=WavePlayer_14, properties {soundWaveAssetPtr=&/Henchman_Interrogation_060}}
 {id=WavePlayer_15, properties {soundWaveAssetPtr=&/Henchman_Interrogation_070}}
 {id=WavePlayer_16, properties {soundWaveAssetPtr=&/Henchman_Interrogation_080}}
 {id=WavePlayer_17, properties {soundWaveAssetPtr=&/Henchman_Interrogation_090}}
]
properties {
 attenuationSettings=&/Henchman_MANG_Close_Att
 maxDistance=4000
 soundClassObject=/Game/Sounds/FortSoundClassesAndModes/Fort_VO_Enemy_NonCritical
 volumeMultiplier=1.2
 concurrencySet [&/MANG_Chatter_Conc]
}
